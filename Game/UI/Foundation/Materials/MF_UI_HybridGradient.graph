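# UI material function that outputs a hybrid gradient color, layering two radial gradients over a rotatable two-color linear gradient.
# MF_UI_LinearGradient blends the two base colors at the given rotation angle.
# Each MF_UI_RadialGradient call samples UVs shifted by its X/Y offsets, then lerps its RGB over the result below, weighted by that color's alpha.
# Every input has a preview default, and the second radial gradient sits on top.
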MaterialFunction /Game/UI/Foundation/Materials/MF_UI_HybridGradient
graph {nodes [
  n1 = FunctionInput() {bUsePreviewValueAsDefault=true, inputName=LinearGradientColorA, sortPriority=1, previewValue=<0.0356 0 0.3325 1>}
  n2 = FunctionInput() {bUsePreviewValueAsDefault=true, inputName=LinearGradientColorB, sortPriority=2, previewValue=<0.2623 0.117 1 1>}
  n3 = FunctionInput() {
   bUsePreviewValueAsDefault=true
   inputName=LinearGradientRotationAngle
   inputType=FunctionInput_Scalar
   sortPriority=3
   previewValue=<0.25 0 0 0>
  }
  n4 = LinearInterpolate(
   A=@n1
   B=@n2
   Alpha=MaterialFunctionCall(RotationAngle=@n3) {materialFunction=&MF_UI_LinearGradient}.Result
  )
  n5 = FunctionInput() {
   bUsePreviewValueAsDefault=true
   inputName=RadialGradientOneColor
   inputType=FunctionInput_Vector4
   sortPriority=4
   previewValue=<0.9647 0 0.2423 0.2>
  }
  n6 = ComponentMask(None=@n5) {b=true, g=true, r=true}
  n7 = FunctionInput() {bUsePreviewValueAsDefault=true, inputName=RadialGradientOneOffsetX, inputType=FunctionInput_Scalar, sortPriority=5, previewValue=<0.5 0 0 0>}
  n8 = FunctionInput() {bUsePreviewValueAsDefault=true, inputName=RadialGradientOneOffsetY, inputType=FunctionInput_Scalar, sortPriority=6, previewValue=<0.5 0 0 0>}
  n9 = AppendVector(A=@n7, B=@n8)
  n10 = TextureCoordinate()
  n11 = Multiply(
   A=ComponentMask(None=@n5) {a=true}
   B=MaterialFunctionCall(UV=Add(A=@n9, B=@n10)) {materialFunction=&MF_UI_RadialGradient}.Result
  )
  n12 = FunctionInput() {
   bUsePreviewValueAsDefault=true
   inputName=RadialGradientTwoColor
   inputType=FunctionInput_Vector4
   sortPriority=7
   previewValue=<0 1 0.8632 0.5>
  }
  n13 = ComponentMask(None=@n12) {a=true}
  n14 = FunctionInput() {
   bUsePreviewValueAsDefault=true
   inputName=RadialGradientTwoOffsetX
   inputType=FunctionInput_Scalar
   sortPriority=8
   previewValue=<-0.25 0 0 0>
  }
  n15 = FunctionInput() {bUsePreviewValueAsDefault=true, inputName=RadialGradientTwoOffsetY, inputType=FunctionInput_Scalar, sortPriority=9, previewValue=<0.5 0 0 0>}
  n16 = MaterialFunctionCall(UV=Add(A=@n10, B=AppendVector(A=@n14, B=@n15))) {materialFunction=&MF_UI_RadialGradient}
  FunctionOutput(None=LinearInterpolate(A=LinearInterpolate(A=@n4, B=@n6, Alpha=@n11), B=ComponentMask(None=@n12) {b=true, g=true, r=true}, Alpha=Multiply(A=@n13, B=@n16.Result)))
 ]}
