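# UI material function that procedurally draws a bullseye target mask, filed under the Reboot Rally sidebar UI.
# Centre-scaled, fill-fitted UI UVs drive several MF_UI_Circle SDFs at different radii, mostly outlines plus one filled shape.
# A narrow SmoothStep sharpens each shape, then nested lerps and inversions merge them into one output without texture samples.
# Texture scale and line thickness are exposed as inputs with defaults.
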MaterialFunction /Game/UI/Sidebar/RebootRally/Materials/MF_UI_Target_Bullseye
graph {nodes [
  n1 = NamedRerouteUsage()
  n2 = MaterialFunctionCall(UVs=NamedRerouteUsage(), Size=Constant() {r=0.95}, OutlineThickness=NamedRerouteUsage()) {materialFunction=&MF_UI_Circle}
  n3 = NamedRerouteUsage()
  n4 = MaterialFunctionCall(UVs=NamedRerouteUsage(), Size=Constant() {r=1.3}, OutlineThickness=NamedRerouteUsage()) {materialFunction=&MF_UI_Circle}
  n5 = NamedRerouteUsage()
  n6 = Constant() {r=0.75}
  n7 = NamedRerouteUsage()
  n8 = LinearInterpolate(
   B=LinearInterpolate(
    B=SmoothStep(Max=@n1, Value=@n2.SDFOutline) {constMax=0.002, constValue=1}
    Alpha=SmoothStep(Max=@n3, Value=@n4.SDFOutline) {constMax=0.002, constValue=1}
   )
   Alpha=SmoothStep(
    Max=NamedRerouteUsage()
    Value=MaterialFunctionCall(UVs=@n5, Size=@n6, OutlineThickness=@n7) {materialFunction=&MF_UI_Circle}.SDFOutline
   ) {constMax=0.002, constValue=1}
  )
  n9 = NamedRerouteUsage()
  n10 = Constant() {r=0.55}
  n11 = SmoothStep(
   Max=NamedRerouteUsage()
   Value=MaterialFunctionCall(UVs=NamedRerouteUsage(), Size=Constant() {r=0.35}, OutlineThickness=NamedRerouteUsage()) {materialFunction=&MF_UI_Circle}.SDFOutline
  ) {constMax=0.002, constValue=1}
  n12 = SmoothStep(
   Max=NamedRerouteUsage()
   Value=MaterialFunctionCall(UVs=NamedRerouteUsage(), Size=Constant() {r=0.15}, OutlineThickness=NamedRerouteUsage()) {materialFunction=&MF_UI_Circle}.SDFOutline
  ) {constMax=0.002, constValue=1}
  n13 = LinearInterpolate(
   A=SmoothStep(
    Max=NamedRerouteUsage()
    Value=MaterialFunctionCall(UVs=@n9, Size=@n10) {materialFunction=&MF_UI_Circle}.SDF
   ) {constMax=0.002, constValue=1}
   Alpha=LinearInterpolate(A=OneMinus(None=@n11), Alpha=OneMinus(None=@n12))
  )
  FunctionOutput(None=OneMinus(None=LinearInterpolate(B=@n8, Alpha=@n13)))
  n14 = MaterialFunctionCall() {materialFunction=&GetUserInterfaceUV}
  n15 = FunctionInput(Preview=Constant() {r=1}) {bUsePreviewValueAsDefault=true, inputName=Texture Sclae, inputType=FunctionInput_Scalar}
  NamedRerouteDeclaration(None=MaterialFunctionCall(UV=MaterialFunctionCall(UVs=@n14.Normalized UV, Texture Scale=@n15) {materialFunction=&ScaleUVsByCenter}.UVs) {materialFunction=&MF_UI_UVScaling}.Scaled To Fill) {name=UV, nodeColor=<rgba 0 0 0 1>}
  NamedRerouteDeclaration(None=Constant() {r=0.005}) {name=SmoothStep Max Value, nodeColor=<rgba 0 0 0 1>}
  NamedRerouteDeclaration(None=FunctionInput(Preview=Constant() {r=0.005}) {bUsePreviewValueAsDefault=true, inputName=Bullseye Line Thickness, inputType=FunctionInput_Scalar, sortPriority=10}) {name=Outline Thickness, nodeColor=<rgba 0 0 0 1>}
 ]}
