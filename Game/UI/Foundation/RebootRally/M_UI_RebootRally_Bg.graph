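# Translucent UI material for the Reboot Rally background, built on a hybrid linear and two-radial gradient.
# Static switches layer in color-dodged grunge with a bullseye, panning decorative text with stepped per-column speeds, bullet holes, and a grunge-broken box outline.
# An angled, optionally animated left bar color is composited last.
# A separate mask takes the minimum of a skewed square SDF and two linear fades, each switchable.
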Material /Game/UI/Foundation/RebootRally/M_UI_RebootRally_Bg
graph {
 properties {
  bUsedWithStaticMesh=false
  blendMode=BLEND_Translucent
  materialDomain=MD_UI
  pixelDepthOffsetMode=PDOM_Legacy
  usedShadingModels=Default Lit
 }
 nodes [
  n1 = ConstantBiasScale(None=MaterialFunctionCall(
    UVs=TextureCoordinate()
    "Rotation Angle (0-1)"=ScalarParameter() {defaultValue=0.5, group=Skew, parameterName=Skew Rotation, sliderMax=1, sortPriority=36}
   ) {materialFunction=&CustomRotator}.Rotated Values) {bias=-0.5, scale=2}
  n2 = LinearInterpolate(
   B=Multiply(
    A=ComponentMask(None=@n1) {r=true}
    B=ScalarParameter() {defaultValue=1, group=Skew, parameterName=Skew Intensity, sliderMax=1, sortPriority=35}
   )
   Alpha=Saturate(None=ScalarParameter() {defaultValue=0.2, group=Skew, parameterName=Skew Amount, sliderMax=1, sortPriority=34})
  ) {constA=1}
  n3 = Multiply(A=ScalarParameter() {group=Skew, parameterName=Skew Position X, sliderMax=1, sliderMin=-1, sortPriority=39}) {constB=-1}
  n4 = ScalarParameter() {group=Skew, parameterName=Skew Position Y, sliderMax=1, sliderMin=-1, sortPriority=40}
  n5 = AppendVector(
   A=ScalarParameter() {defaultValue=0.4, group=Skew, parameterName=Skew Width, sortPriority=37}
   B=ScalarParameter() {defaultValue=0.4, group=Skew, parameterName=Skew Height, sortPriority=38}
  )
  MaterialExpressionSubtract_0 = Subtract(
   A=Add(
    A=MaterialFunctionCall(Texture Scale=@n2) {materialFunction=&ScaleUVsByCenter}.UVs
    B=AppendVector(A=@n3, B=@n4)
   )
   B=OneMinus(None=@n5)
  )
  MaterialExpressionSubtract_1 = Subtract(A=@n5) {constB=0.1}
  n6 = Custom!(p=@MaterialExpressionSubtract_0, b=@MaterialExpressionSubtract_1) {
   code="float2 d = abs(p)-b;\r\nreturn length(max(d,0.0)) + min(max(d.x,d.y),0.0);"
   description=Square SDF
   inputs [
    {inputName=p, input {expression=&self:MaterialExpressionSubtract_0}}
    {inputName=b, input {expression=&self:MaterialExpressionSubtract_1}}
   ]
  }
  n7 = OneMinus(None=SmoothStep(Max=ScalarParameter() {defaultValue=0.01, group=Skew, parameterName=Skew Feather, sliderMax=1, sortPriority=39}, Value=@n6) {constMax=0.01})
  n8 = Constant() {r=1}
  n9 = MaterialFunctionCall(
   GradientStart=ScalarParameter() {group=Fade, parameterName=Fade Start, sortPriority=34}
   GradientEnd=ScalarParameter() {defaultValue=1, group=Fade, parameterName=Fade End, sortPriority=35}
   RotationAngle=ScalarParameter() {group=Fade, parameterName=Fade Rotation, sliderMax=1, sortPriority=36}
  ) {materialFunction=&MF_UI_LinearGradient}
  n10 = ScalarParameter() {group=Right Fade, parameterName=R Fade Start, sortPriority=34}
  n11 = ScalarParameter() {defaultValue=1, group=Right Fade, parameterName=R Fade End, sortPriority=35}
  n12 = ScalarParameter() {group=Right Fade, parameterName=R Fade Rotation, sliderMax=1, sortPriority=36}
  Min(
   A=Min(
    A=StaticSwitchParameter(True=@n7, False=@n8) {group=Skew, parameterName=Enable Skew, sortPriority=33}
    B=StaticSwitchParameter(True=@n9.Result, False=@n8) {group=Fade, parameterName=Enable Fade, sortPriority=33}
   )
   B=StaticSwitchParameter(
    True=MaterialFunctionCall(GradientStart=@n10, GradientEnd=@n11, RotationAngle=@n12) {materialFunction=&MF_UI_LinearGradient}.Result
    False=@n8
   ) {group=Right Fade, parameterName=Enable R Fade, sortPriority=33}
  )
  n13 = MaterialFunctionCall(
   UVs=MaterialFunctionCall() {materialFunction=&GetUserInterfaceUV}.Normalized UV
   "Rotation Angle (0-1)"=Constant() {r=-0.002}
  ) {materialFunction=&CustomRotator}
  n14 = Constant2Vector() {g=1.03, r=1.03}
  n15 = ScalarParameter() {defaultValue=0.05, group=Outline, parameterName=Outline Thickness}
  n16 = StaticBool() {value=true}
  n17 = MaterialFunctionCall() {materialFunction=&GetUserInterfaceUV}
  n18 = ScalarParameter() {defaultValue=0.3, group=Outline, parameterName=Mask Scale}
  n19 = TextureSample(UVs=MaterialFunctionCall(UV=MaterialFunctionCall(UVs=@n17.Normalized UV, Texture Scale=@n18) {materialFunction=&ScaleUVsByCenter}.UVs) {materialFunction=&MF_UI_UVScaling}.Scaled To Fill) {samplerType=SAMPLERTYPE_Alpha, texture=/Game/UI/Sidebar/RebootRally/Textures/S28_Textures/T_UI_Seamless_Grunge}
  n20 = Multiply(A=@n19.RGB, B=@n19.RGB)
  n21 = LinearInterpolate(
   B=LinearInterpolate(
    B=VectorParameter() {group=Outline, parameterName=Outline Color, defaultValue=<rgba 1 1 1 1>}.RGB
    Alpha=MaterialFunctionCall(UVs=@n13.Rotated Values, Size=@n14, OutlineThickness=@n15, KeepAspectRatio=@n16) {materialFunction=&MF_UI_Box}.Outline
   )
   Alpha=Power(Base=Multiply(A=@n20, B=@n19.RGB)) {constExponent=4}
  )
  n22 = ScalarParameter() {defaultValue=1, group=Outline, parameterName=Outline Opacity}
  NamedRerouteDeclaration(None=Add(A=Multiply(A=@n21, B=@n22), B=NamedRerouteUsage())) {name=Add Outline, nodeColor=<rgba 0 0 0 1>}
  n23 = NamedRerouteUsage()
  n24 = MaterialFunctionCall(
   UVs=MaterialFunctionCall() {materialFunction=&GetUserInterfaceUV}.9-Slice UV
   Texture Scale=ScalarParameter() {defaultValue=1, group=Static Texture, parameterName=Scale, sortPriority=3}
  ) {materialFunction=&ScaleUVsByCenter}
  n25 = TextureObjectParameter() {
   group=Static Texture
   parameterName=Static Texture
   samplerType=SAMPLERTYPE_Masks
   sortPriority=98
   texture=/EpicBaseTextures/Grunges/T_Micro_Detail_MSK
  }
  n26 = StaticComponentMaskParameter(None=TextureSample(UVs=MaterialFunctionCall(UV=@n24.UVs) {materialFunction=&MF_UI_UVScaling}.Scaled To Fill, Tex=@n25).RGB) {defaultG=true, group=Static Texture, parameterName=Texture Channel, sortPriority=99}
  n27 = Power(
   Base=Multiply(A=@n26, B=@n26)
   Exp=ScalarParameter() {defaultValue=1, group=Static Texture, parameterName=Adjust Levels, sliderMax=3, sortPriority=4}
  )
  n28 = ScalarParameter() {defaultValue=1, group=Static Texture, parameterName=Texture Opacity, sliderMax=1, sortPriority=2}
  n29 = Max(A=Saturate(None=Multiply(A=@n27, B=@n28))) {constB=0}
  n30 = VectorParameter() {parameterName=RadialGradientOneColor, defaultValue=<rgba 0.9647 0 0.2423 0.2>}
  n31 = VectorParameter() {parameterName=RadialGradientTwoColor, defaultValue=<rgba 0 1 0.8632 0.5>}
  n32 = MaterialFunctionCall!(
   LinearGradientColorA=VectorParameter() {parameterName=LinearGradientColorA, defaultValue=<rgba 0.0356 0 0.3325 1>}.RGB
   LinearGradientColorB=VectorParameter() {parameterName=LinearGradientColorB, defaultValue=<rgba 0.2623 0.117 1 1>}.RGB
   LinearGradientRotationAngle=ScalarParameter() {defaultValue=0.25, parameterName=LinearGradientRotationAngle}
   RadialGradientOneColor=AppendVector(A=@n30.RGB, B=@n30.RGB)
   RadialGradientOneOffsetX=ScalarParameter() {defaultValue=0.5, parameterName=RadialGradientOneOffsetX}
   RadialGradientOneOffsetY=ScalarParameter() {defaultValue=0.5, parameterName=RadialGradientOneOffsetY}
   RadialGradientTwoColor=AppendVector(A=@n31.RGB, B=@n31.RGB)
   RadialGradientTwoOffsetX=ScalarParameter() {defaultValue=-0.25, parameterName=RadialGradientTwoOffsetX}
   RadialGradientTwoOffsetY=ScalarParameter() {defaultValue=0.5, parameterName=RadialGradientTwoOffsetY}
  ) {materialFunction=/Game/UI/Foundation/Materials/MF_UI_HybridGradient}
  n33 = MaterialFunctionCall!(Blend=@n29, Base=@n32.Result) {materialFunction=/Engine/Functions/Engine_MaterialFunctions03/Blends/Blend_ColorDodge}
  n34 = StaticSwitchParameter(
   True=StaticSwitchParameter(
    True=MaterialFunctionCall(Blend=@n23, Base=@n33.Result) {materialFunction=&Blend_ColorDodge}.Result
    False=@n33.Result
   ) {group=Static Texture, parameterName=Enable Bullseye}
   False=@n32.Result
  ) {group=Static Texture, parameterName=Enable Static Texture, sortPriority=1}
  NamedRerouteDeclaration(None=@n34) {name=Base after texture, nodeColor=<rgba 0 0 0 1>}
  n35 = ScalarParameter() {defaultValue=0.005, group=Static Texture, parameterName=Bullseye Line Thickness, sliderMax=0.05, sliderMin=0.001, sortPriority=34}
  n36 = MaterialFunctionCall!(
   Texture Sclae=ScalarParameter() {defaultValue=1, group=Static Texture, parameterName=Bullseye Scale, sortPriority=35}
   Bullseye Line Thickness=@n35
  ) {materialFunction=/Game/UI/Sidebar/RebootRally/Materials/MF_UI_Target_Bullseye}
  n37 = ScalarParameter() {defaultValue=0.5, group=Static Texture, parameterName=Bullseye Opacity, sortPriority=33}
  NamedRerouteDeclaration(None=LinearInterpolate(B=Multiply(A=@n36.Result, B=@n37) {constB=0.5}, Alpha=@n29)) {name=Bullseye, nodeColor=<rgba 0 0 0 1>}
  n38 = Add(
   A=MaterialFunctionCall() {materialFunction=&GetUserInterfaceUV}.Normalized UV
   B=AppendVector(
    A=ScalarParameter() {group=Floating Texture, parameterName=X2, sortPriority=5}
    B=ScalarParameter() {group=Floating Texture, parameterName=Y2, sortPriority=6}
   )
  )
  n39 = ScalarParameter() {defaultValue=0.08, group=Floating Texture, parameterName=Floating Texture 2 Scale, sortPriority=4}
  n40 = TextureSample(UVs=MaterialFunctionCall(UV=MaterialFunctionCall(UVs=@n38, Texture Scale=@n39) {materialFunction=&ScaleUVsByCenter}.UVs) {materialFunction=&MF_UI_UVScaling}.Scaled To Fill) {texture=/Game/UI/Sidebar/RebootRally/Textures/S28_Textures/T_UI_BulletHole1}
  n41 = NamedRerouteUsage()
  n42 = OneMinus(None=@n40.A)
  n43 = Add(
   A=MaterialFunctionCall() {materialFunction=&GetUserInterfaceUV}.Normalized UV
   B=AppendVector(
    A=ScalarParameter() {group=Floating Texture, parameterName=X1, sortPriority=2}
    B=ScalarParameter() {group=Floating Texture, parameterName=Y1, sortPriority=3}
   )
  )
  n44 = ScalarParameter() {defaultValue=0.15, group=Floating Texture, parameterName="Floating Texture1  Scale", sortPriority=1}
  n45 = TextureSample(UVs=MaterialFunctionCall(UV=MaterialFunctionCall(UVs=@n43, Texture Scale=@n44) {materialFunction=&ScaleUVsByCenter}.UVs) {materialFunction=&MF_UI_UVScaling}.Scaled To Fill) {texture=/Game/UI/Sidebar/RebootRally/Textures/S28_Textures/T_UI_BulletHole2}
  n46 = NamedRerouteUsage()
  n47 = OneMinus(None=@n45.A)
  n48 = LinearInterpolate(A=@n47, B=@n42, Alpha=Add(A=@n47, B=@n42))
  NamedRerouteDeclaration(None=LinearInterpolate(
    A=LinearInterpolate(A=@n40.RGB, B=@n41, Alpha=@n42)
    B=LinearInterpolate(A=@n45.RGB, B=@n46, Alpha=@n47) {constA=1}
    Alpha=Saturate(None=@n48)
   )) {name=BulletHoles, nodeColor=<rgba 0 0 0 1>}
  n49 = Constant2Vector() {g=1, r=3}
  NamedRerouteDeclaration(None=@n49) {name=Text Tile Value, nodeColor=<rgba 0 0 0 1>}
  n50 = VectorParameter() {group=Left Bar, parameterName=Left Bar Color, sortPriority=36, defaultValue=<rgba 0.9301 1 0.0103 1>}
  n51 = ScalarParameter() {defaultValue=-0.25, group=Left Bar, parameterName=Left Bar Position, sortPriority=37}
  n52 = Multiply(A=Abs(None=@n51)) {constB=0.5}
  n53 = ScalarParameter() {group=Left Bar, parameterName=Left Bar Animation Transition, sliderMax=1, sortPriority=35}
  n54 = Multiply(A=@n53) {constB=2}
  n55 = LinearInterpolate(B=Multiply(A=Subtract(A=@n53) {constB=0.5}) {constB=2}, Alpha=Saturate(None=Floor(None=@n54))) {desc="AnimateHIghlight <= 0.5 result is 0"}
  n56 = Add(
   A=StaticSwitchParameter(True=LinearInterpolate(A=@n52, B=@n51, Alpha=@n55) {constA=-0.1, constAlpha=1}, False=@n51) {defaultValue=true, group=Left Bar, parameterName=Enable Left Bar Animation, sortPriority=34}
   B=MaterialFunctionCall() {materialFunction=&GetUserInterfaceUV}.9-Slice UV
  )
  n57 = ScalarParameter() {defaultValue=-0.25, group=Left Bar, parameterName=Left Bar Angle, sortPriority=38}
  n58 = Multiply(
   A=ComponentMask(None=Rotator(Coordinate=@n56, Time=@n57)) {r=true}
   B=ScalarParameter() {defaultValue=256, group=Left Bar, parameterName=Left Bar Sharpness, sortPriority=39}
  )
  n59 = ScalarParameter() {defaultValue=1, group=Panning Image, parameterName=Panning Image Opacity, sliderMax=1, sortPriority=35}
  n60 = Multiply(
   A=MaterialFunctionCall() {materialFunction=&GetUserInterfaceUV}.Normalized UV
   B=ScalarParameter() {defaultValue=1, group=Panning Image, parameterName=Pan Image Scale}
  )
  n61 = AppendVector(A=ScalarParameter() {group=Panning Image, parameterName=Pan Image Offset}, B=Constant())
  n62 = NamedRerouteDeclaration(None=MaterialFunctionCall(UV=Subtract(A=@n60, B=@n61), TextureSize=AppendVector(A=@n49.G, B=@n49.G)) {materialFunction=&MF_UI_UVScaling}.Scaled To Fill) {name=Text UV, nodeColor=<rgba 0 0 0 1>}
  n63 = ComponentMask(None=@n62) {r=true}
  n64 = OneMinus(None=@n63)
  n65 = Saturate(None=Multiply(A=Min(A=@n63, B=@n64)) {constB=2})
  n66 = Constant()
  n67 = Time()
  n68 = NamedRerouteUsage()
  n69 = NamedRerouteUsage()
  n70 = MaterialFunctionCall(
   A=ScalarParameter() {defaultValue=-0.015, group=Panning Image, parameterName=Pan Image Col 1 Speed}
   B=ScalarParameter() {defaultValue=0.01, group=Panning Image, parameterName=Pan Image Col 2 Speed}
   C=ScalarParameter() {defaultValue=-0.01, group=Panning Image, parameterName=Pan Image Col 3 Speed}
   Alpha=MaterialFunctionCall!("Value (0-1)"=ComponentMask(None=@n68) {r=true}, Steps=ComponentMask(None=@n69) {r=true}) {materialFunction=/FortUILibrary/MaterialFunctions/Basics/MF_UI_StepRatio}.Stepped Gradient
  ) {materialFunction=&3ColorBlend}
  n71 = Add(A=Frac(None=Multiply(A=@n67, B=@n70.Result)), B=Multiply(A=@n70.Result) {constB=5})
  n72 = Add(A=@n62, B=Frac(None=AppendVector(A=@n66, B=@n71))) {constB=0}
  n73 = MaterialFunctionCall(Float2=NamedRerouteUsage()) {materialFunction=&BreakOutFloat2Components}
  n74 = TextureObjectParameter() {
   group=Panning Image
   parameterName=Panning Image Texture
   samplerType=SAMPLERTYPE_Alpha
   sortPriority=34
   texture=/Game/UI/Sidebar/RebootRally/Textures/T_UI_RebootRally_DecorativeText
  }
  n75 = TextureSample(
   UVs=MaterialFunctionCall!(UV=Frac(None=@n72), Tiling X=@n73.R, Tiling Y=@n73.R) {materialFunction=/FortUILibrary/MaterialFunctions/Basics/MF_UI_Pattern}.Pattern UVs
   Tex=@n74
  ) {
   samplerSource=SSM_Clamp_WorldGroupSettings
   samplerType=SAMPLERTYPE_Alpha
   texture=&T_UI_RebootRally_DecorativeText
  }
  n76 = NamedRerouteDeclaration(None=Multiply(A=Ceil(None=@n65), B=@n75.A)) {name=Text Mask, nodeColor=<rgba 0 0 0 1>}
  n77 = Add(A=@n32.Result, B=Multiply(A=@n32.Result, B=Multiply(A=@n59, B=@n76)))
  n78 = LinearInterpolate(
   A=StaticSwitchParameter(True=Saturate(None=@n77), False=@n32.Result) {group=Panning Image, parameterName=Enable Panning Image, sortPriority=33}
   Alpha=@n34
  )
  n79 = StaticSwitchParameter(True=NamedRerouteUsage(), False=@n78) {group=Floating Texture, parameterName=Enable Floating Texture, sortPriority=0}
  n80 = StaticSwitchParameter(True=NamedRerouteUsage(), False=@n78) {group=Outline, parameterName=Enable Outline}
  n81 = LinearInterpolate(A=@n79, B=@n80, Alpha=Max(A=@n80, B=@n79))
  StaticSwitchParameter(True=LinearInterpolate(A=@n50.RGB, B=@n81, Alpha=Saturate(None=@n58)), False=@n81) {group=Left Bar, parameterName=Enable Left Bar, sortPriority=33}
 ]
}
